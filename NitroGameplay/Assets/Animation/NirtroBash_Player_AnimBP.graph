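# Player animation Blueprint for the Nitro bash moves, implementing the MovementMode layer interface with logic only in FinalPoseOverride.
# That layer blends between the incoming pose and a state machine with Wait, two Bash states and Bounce, driven by boolean flags.
# Bash 01 layers over the incoming pose from Spine_01, Bash 02 and Bounce replace it, and Bounce can interrupt any state.
# State entry functions maintain bIsCurrentlyBashing, clearing it on returning to Wait.
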AnimBlueprint /NitroGameplay/Assets/Animation/NirtroBash_Player_AnimBP
target_skeleton=&Fortnite_M_Avg_Player_Skeleton
graphs [
 AnimGraph {nodes [Output Pose()]}
 EventGraph {links [], nodes []}
 MovementMode_AimOffset {nodes [
   Input Pose() {name=InAimOffsetsPose}
   Input Pose() {name=InSourcePosePreAimOffsets}
   Output Pose() {layerGroup=MovementMode}
  ]}
 MovementMode_BaseLocomotion {nodes [
   Input Pose() {name=InPoseBaseLocomotion}
   Output Pose() {layerGroup=MovementMode}
  ]}
 MovementMode_FinalPoseOverride {nodes [
   AnimGraphNode_StateMachine_0 = TacSprint SM() {is=title AnimGraphNode_StateMachine}
   Output Pose(Result=Blend Poses by bool(BlendPose_0=@AnimGraphNode_StateMachine_0, BlendPose_1=Use cached pose 'FinalInput Pose'(), BlendTime_0=0, BlendTime_1=1) {blendType=ExpIn}) {layerGroup=MovementMode}
   Save cached pose 'FinalInput Pose'(Pose=Input Pose() {name=InFinalPose}) {cacheName=FinalInput Pose}
  ]}
 TacSprint SM {
  entry=Wait
  path=MovementMode_FinalPoseOverride.AnimGraphNode_StateMachine_0.TacSprint SM
  ids=id
  nodes [
   To Bash 02() {is=title AnimStateAliasNode, also {aliases [Wait]}}
   To Bash 01() {is=title AnimStateAliasNode, also {aliases [Wait]}}
   To Bounce() {is=title AnimStateAliasNode, bGlobalAlias=true}
   AnimStateNode_0 = Wait() {is=title AnimStateNode}
   AnimStateNode_1 = Bash 02() {is=title AnimStateNode}
   AnimStateNode_2 = Bash 01() {is=title AnimStateNode}
   AnimStateNode_3 = Bounce() {is=title AnimStateNode}
   AnimStateTransitionNode_0 = AnimStateTransitionNode() {bAutomaticRuleBasedOnSequencePlayerInState=true}
   AnimStateTransitionNode_1 = AnimStateTransitionNode()
   AnimStateTransitionNode_2 = AnimStateTransitionNode() {crossfadeDuration=0.1}
   AnimStateTransitionNode_3 = AnimStateTransitionNode()
   AnimStateTransitionNode_4 = AnimStateTransitionNode() {crossfadeDuration=0.1}
   AnimStateTransitionNode_5 = AnimStateTransitionNode() {crossfadeDuration=0.25}
  ]
  links [
   {
    dst=Wait
    rule=MovementMode_FinalPoseOverride.AnimGraphNode_StateMachine_0.TacSprint SM.AnimStateTransitionNode_3.Transition
    src=Bash 01
    via=AnimStateTransitionNode_3
   }
   {
    dst=Wait
    rule=MovementMode_FinalPoseOverride.AnimGraphNode_StateMachine_0.TacSprint SM.AnimStateTransitionNode_5.Transition
    src=Bash 02
    via=AnimStateTransitionNode_5
   }
   {
    dst=Wait
    rule=MovementMode_FinalPoseOverride.AnimGraphNode_StateMachine_0.TacSprint SM.AnimStateTransitionNode_0.Transition
    src=Bounce
    via=AnimStateTransitionNode_0
   }
   {
    dst=Bash 01
    rule=MovementMode_FinalPoseOverride.AnimGraphNode_StateMachine_0.TacSprint SM.AnimStateTransitionNode_2.Transition
    src=To Bash 01
    via=AnimStateTransitionNode_2
   }
   {
    dst=Bash 02
    rule=MovementMode_FinalPoseOverride.AnimGraphNode_StateMachine_0.TacSprint SM.AnimStateTransitionNode_4.Transition
    src=To Bash 02
    via=AnimStateTransitionNode_4
   }
   {
    dst=Bounce
    rule=MovementMode_FinalPoseOverride.AnimGraphNode_StateMachine_0.TacSprint SM.AnimStateTransitionNode_1.Transition
    src=To Bounce
    via=AnimStateTransitionNode_1
   }
  ]
 }
 Wait {
  path=MovementMode_FinalPoseOverride.AnimGraphNode_StateMachine_0.TacSprint SM.AnimStateNode_0.Wait
  nodes [Output Animation Pose(Result=Use cached pose 'FinalInput Pose'()) {
    node {stateEntryFunction {functionName=OnStateEntry}}
    stateEntryFunction {bSelfContext=true, memberGuid=#1, memberName=OnBashWaitStateEntry}
   }]
 }
 Bash 02 {
  path=MovementMode_FinalPoseOverride.AnimGraphNode_StateMachine_0.TacSprint SM.AnimStateNode_1.Bash 02
  nodes [Output Animation Pose(Result=Sequence Player!(Sequence:Anim Sequence Base Object Reference) -> (Pose) {bLoopAnimation=false}) {
    node {stateEntryFunction {functionName=OnBashStateEntry}}
    stateEntryFunction {bSelfContext=true, memberGuid=#2, memberName=OnBashStateEntry}
   }]
 }
 Bash 01 {
  path=MovementMode_FinalPoseOverride.AnimGraphNode_StateMachine_0.TacSprint SM.AnimStateNode_2.Bash 01
  nodes [
   n1 = Layered blend per bone(BasePose=Use cached pose 'FinalInput Pose'(), BlendPoses_0=Sequence Player() {bLoopAnimation=false}) {bMeshSpaceRotationBlend=true, layerSetup [{branchFilters [{blendDepth=4, boneName=Spine_01}, {boneName=None}]}]}
   Output Animation Pose(Result=@n1) {
    node {stateEntryFunction {functionName=OnBashStateEntry}}
    stateEntryFunction {bSelfContext=true, memberGuid=#2, memberName=OnBashStateEntry}
   }
  ]
 }
 Bounce {
  path=MovementMode_FinalPoseOverride.AnimGraphNode_StateMachine_0.TacSprint SM.AnimStateNode_3.Bounce
  nodes [Output Animation Pose(Result=Sequence Player 'NitroSprint_Bash_Bounce_01'() {bLoopAnimation=false, sequence=&/NitroSprint_Bash_Bounce_01}) {
    node {stateEntryFunction {functionName=OnStateEntry}}
    stateEntryFunction {bSelfContext=true, memberGuid=#2, memberName=OnBashStateEntry}
   }]
 }
 Transition {path=MovementMode_FinalPoseOverride.AnimGraphNode_StateMachine_0.TacSprint SM.AnimStateTransitionNode_0.Transition, nodes [Result()]}
 Transition {
  path=MovementMode_FinalPoseOverride.AnimGraphNode_StateMachine_0.TacSprint SM.AnimStateTransitionNode_1.Transition
  nodes [Result(bCanEnterTransition=Get bIsBounceBack() -> (bIsBounceBack:bool))]
 }
 Transition {
  path=MovementMode_FinalPoseOverride.AnimGraphNode_StateMachine_0.TacSprint SM.AnimStateTransitionNode_2.Transition
  nodes [Result(bCanEnterTransition=Get bIsBash_01() -> (bIsBash_01:bool))]
 }
 Transition {
  path=MovementMode_FinalPoseOverride.AnimGraphNode_StateMachine_0.TacSprint SM.AnimStateTransitionNode_3.Transition
  nodes [Result(bCanEnterTransition=Get bIsExitBash() -> (bIsExitBash:bool))]
 }
 Transition {
  path=MovementMode_FinalPoseOverride.AnimGraphNode_StateMachine_0.TacSprint SM.AnimStateTransitionNode_4.Transition
  nodes [Result(bCanEnterTransition=Get bIsBash_02() -> (bIsBash_02:bool))]
 }
 Transition {
  path=MovementMode_FinalPoseOverride.AnimGraphNode_StateMachine_0.TacSprint SM.AnimStateTransitionNode_5.Transition
  nodes [Result(bCanEnterTransition=Get bIsExitBash())]
 }
 MovementMode_FullBodyOverride {nodes [
   Input Pose() {name=InputPoseFullBody}
   Output Pose() {layerGroup=MovementMode}
  ]}
 MovementMode_InAirOverride {nodes [
   Input Pose() {name=InSource_Pose}
   Input Pose() {name=InAirPose}
   Output Pose() {layerGroup=MovementMode}
  ]}
 MovementMode_LocomotionAdditives {nodes [
   Input Pose() {name=InLocomotionAdditivesPose}
   Output Pose() {layerGroup=MovementMode}
  ]}
 MovementMode_LowerBodyOverride {nodes [
   Input Pose() {name=InLowerBodyPose}
   Input Pose() {name=InSourcePose}
   Output Pose() {layerGroup=MovementMode}
  ]}
 MovementMode_OffScreen {nodes [
   Input Pose() {name=InPoseOffscreen}
   Output Pose() {layerGroup=MovementMode}
  ]}
 MovementMode_SourcePose {nodes [
   Input Pose() {name=InSourcePoseBase}
   Output Pose() {layerGroup=MovementMode}
  ]}
 OnBashStateEntry {nodes [
   OnBashStateEntry!() -> (then, Context:"Anim Update Context Structure (by ref)", Node:"Anim Node Reference Structure (by ref)") {is=title K2Node_FunctionEntry}
   > Set bIsCurrentlyBashing(bIsCurrentlyBashing:bool=true) -> (Output_Get:bool)
   > Return Node()
  ]}
 OnBashWaitStateEntry {nodes [
   OnBashWaitStateEntry!() -> (then, Context:"Anim Update Context Structure (by ref)", Node:"Anim Node Reference Structure (by ref)") {is=title K2Node_FunctionEntry}
   > Set bIsCurrentlyBashing(bIsCurrentlyBashing=false)
   > Return Node()
  ]}
]
variables []
