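; Vehicle sound component for the nitro state that layers nitro audio onto a Valet basic car and can duck its engine volume.
; On BeginPlay it caches the car's audio controller and engine volume, spawns an attached nitro loop, and binds gear changes to play upshift sounds.
; A repeating timer feeds throttle to the loop as a sound parameter.
; Custom events lerp engine volume by alpha, or fade out the loop and destroy the component.
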(event EventBeginPlay
  (Utilities|FlowControl|Sequence
    (:then_0
      ;[0 0.1341 0.5521] Cache stuff
      (bind _asvalet_basic_car_vehicle (Utilities|Casting|CastToValet_BasicCar_Vehicle (Components|GetOwner)))
      (bind _output_get (Variables|Default|SetValet%20Basic%20Car _asvalet_basic_car_vehicle))
      (bind _asbp_vehicle_cosmetics_audio_controller_parent (Utilities|Casting|CastToBP_VehicleCosmetics_AudioController_Parent (Variables|/Script/Engine.ChildActorComponent|GetChildActor (Variables|/Valet/BasicCar/Valet_BasicCar_Vehicle.Valet_BasicCar_Vehicle_C|GetAudioControllerBP _output_get))))
      (bind _output_get_2 (Variables|Default|SetAudio%20Controller _asbp_vehicle_cosmetics_audio_controller_parent))
      (bind _output_get_3 (Variables|Default|SetEngine%20Audio%20Component (AudioController|GetEngineAudioComponent _output_get_2)))
      (Variables|EngineSounds|SetStarting%20Volume%20Multiplier%20Cache (Variables|/Script/Engine.AudioComponent|GetVolumeMultiplier _output_get_3)))
    (:then_1
      (Utilities|FlowControl|Sequence
        (:then_0
          (bind _valet_basic_car (Variables|Default|GetValet%20Basic%20Car)
            (:then
              (bind _returnvalue (Audio|SpawnSoundAttached (Variables|NitroSounds|GetNitro%20Applied%20Loop) (Variables|/Script/Engine.Actor|GetRootComponent _valet_basic_car)))
              (Variables|Default|SetNitro%20Active%20Loop%20Audio _returnvalue)
              (bind _returnvalue_2 (Utilities|Time|SetTimerbyEvent (EventDispatchers|CreateEvent :!event Custom|Update%20Loop%20Audio) 0.1 true))
              (Variables|Default|SetUpdate%20Audio%20Handle _returnvalue_2)
              ;[0.3125 0.2001 0.07579] Gear Changes — currently playing a gear shift sound when shifting into a higher gear in nitro state
              (bind _audio_controller (Variables|Default|GetAudio%20Controller)
                (:then
                  (Private|BindEventtoOnGearChangedEvent _audio_controller (delegate Custom|On%20Gear%20Changed)))
                (:else)))
            (:else)))))))

(event Custom|On%20Gear%20Changed
  (params (NewGear :type Integer) (IgnoreExhaustPops :type Boolean))
  ;[0.3125 0.2001 0.07579] Gear Changes — currently playing a gear shift sound when shifting into a higher gear in nitro state
  (if (> NewGear (Variables|Default|GetPrevGear))
    (bind _valet_basic_car (Variables|Default|GetValet%20Basic%20Car)
      (:then
        (Audio|SpawnSoundAttached (Variables|NitroSounds|GetUpshift%20Sound) (Variables|/Script/Engine.Actor|GetRootComponent _valet_basic_car)))
      (:else)))
  (Variables|Default|SetPrevGear NewGear))

(event Custom|Duck%20Engine%20Volume
  (params (Alpha :type Float%20%28double-precision%29))
  ;[0.08029 0.1927 0.09884] Changes the volume of the vehicle's engine audio component.
  (bind _engine_audio_component (Variables|Default|GetEngine%20Audio%20Component)
    (:then
      (Audio|Components|Audio|SetVolumeMultiplier _engine_audio_component (CallFunction|LerpToTargetVolumeMultiplier Alpha (Variables|EngineSounds|GetEngine%20Volume%20Multiplier%20with%20Nitro)) :!class /Script/Engine.AudioComponent))
    (:else)))

(event Custom|Reset%20Engine%20Audio%20and%20Destroy
  (Audio|Components|Audio|FadeOut (Variables|Default|GetNitro%20Active%20Loop%20Audio) 0.25 0.0 "Linear" :!class /Script/Engine.AudioComponent)
  (Utilities|FlowControl|Delay 0.25)
  (Utilities|Time|ClearandInvalidateTimerbyHandle (Variables|Default|GetUpdate%20Audio%20Handle))
  (Components|DestroyComponent))

(event Custom|Update%20Loop%20Audio
  (bind _nitro_active_loop_audio (Variables|Default|GetNitro%20Active%20Loop%20Audio)
    (:then
      (bind _audio_controller (Variables|Default|GetAudio%20Controller)
        (:then
          (Audio|Parameter|SetFloatParameter _nitro_active_loop_audio "Fort.Vehicle.Throttle" (AudioController|GetThrottleNative _audio_controller) :!class /Script/AudioExtensions.AudioParameterControllerInterface))
        (:else)))
    (:else)))

(fn LerptoTargetVolumeMultiplier
  (params (Alpha :type Float%20%28double-precision%29) (Target :type Float%20%28double-precision%29))
  (return (- (Variables|EngineSounds|GetStarting%20Volume%20Multiplier%20Cache) (* (- (Variables|EngineSounds|GetStarting%20Volume%20Multiplier%20Cache) (Variables|GetTarget)) (Variables|GetAlpha)))))
