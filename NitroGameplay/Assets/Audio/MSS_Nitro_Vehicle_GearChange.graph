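# One-shot stereo MetaSound Source that plays a vehicle gear change sound for Nitro gameplay.
# On play, it picks a random wave from an array of gear shift variations and triggers a stereo Wave Player.
# A RandomSpreadSimple_StartTimePitch patch adds a slight random pitch shift per play, and the player's finish trigger ends the source.
# It uses the Fort_Vehicles sound class and a Nitro player attenuation asset.
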MetaSoundSource /NitroGameplay/Assets/Audio/MSS_Nitro_Vehicle_GearChange
inputs [
 Sounds.GearChange {type=WaveAsset:Array}
 UE.Source.OnPlay {type=Trigger}
]
interfaces [UE.OutputFormat.Stereo:1.0, UE.Source.OneShot:1.0, UE.Source:1.1]
outputs [
 "UE.OutputFormat.Stereo.Audio:0" {type=Audio}
 "UE.OutputFormat.Stereo.Audio:1" {type=Audio}
 UE.Source.OneShot.OnFinished {type=Trigger}
]
pages [Default {
  input_defaults {
   UE.Source.OnPlay=null
   Sounds.GearChange [
    /NitroGameplay/Assets/Audio/Waves/Nitro_State_Vehicle_GearShift_01
    /NitroGameplay/Assets/Audio/Waves/Nitro_State_Vehicle_GearShift_02
    /NitroGameplay/Assets/Audio/Waves/Nitro_State_Vehicle_GearShift_03
    /NitroGameplay/Assets/Audio/Waves/Nitro_State_Vehicle_GearShift_04
    /NitroGameplay/Assets/Audio/Waves/Nitro_State_Vehicle_GearShift_05
    /NitroGameplay/Assets/Audio/Waves/Nitro_State_Vehicle_GearShift_06
    /NitroGameplay/Assets/Audio/Waves/Nitro_State_Vehicle_GearShift_07
   ]
  }
  nodes [
   ArrayRandomGet_0 = Array.Random Get:WaveAsset:Array(In Array=^Sounds.GearChange, Next=^UE.Source.OnPlay)
   WavePlayer_0 = UE.Wave Player:Stereo(
    Play=@ArrayRandomGet_0.On Next
    Wave Asset=@ArrayRandomGet_0.Value
    Pitch Shift=/FortMetasound/Patches/RandomSpreadSimple_StartTimePitch(In=@ArrayRandomGet_0.On Next) {also {defaults {PitchShiftSpread=0.25}}}.RandomPitchShiftOut
   )
  ]
  links [
   {dst="@UE.Source.OneShot.OnFinished", src=WavePlayer_0.On Finished}
   {dst="@UE.OutputFormat.Stereo.Audio:0", src=WavePlayer_0.Out Left}
   {dst="@UE.OutputFormat.Stereo.Audio:1", src=WavePlayer_0.Out Right}
  ]
 }]
properties {
 assetImportData=&self:AssetImportData
 attenuationSettings=/NitroGameplay/Assets/Audio/Settings/Att_NitroState_Player_Apply
 maxDistance=8300
 numChannels=2
 outputFormat=Stereo
 sampleRate=48000
 soundClassObject=&Fort_Vehicles
 volume=0.2
}
