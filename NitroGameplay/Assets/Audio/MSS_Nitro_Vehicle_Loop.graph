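# Stereo MetaSound for the Nitro vehicle-state loop, layering a base loop and a fast loop.
# Smoothed throttle times clamped speed pitches both loops, while movement speed sets their volume above a floor.
# The base loop mixes dry with a low-passed, throttle-scaled copy.
# A threshold-gated Trigger Once starts the fast loop from twice-clamped speed; gear-shift waves and a random-spread patch are unwired.
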MetaSoundSource /NitroGameplay/Assets/Audio/MSS_Nitro_Vehicle_Loop
inputs [
 Fort.Vehicle.Acceleration {type=Float}
 Fort.Vehicle.BlowoffStrength {type=Float}
 Fort.Vehicle.BonusSpeed {type=Float}
 Fort.Vehicle.EnableShifting {type=Bool}
 Fort.Vehicle.Gear {type=Int32}
 Fort.Vehicle.IsGrounded {type=Bool}
 Fort.Vehicle.Load {type=Float}
 Fort.Vehicle.LowFuel {type=Bool}
 Fort.Vehicle.NormalSpeed {type=Float}
 Fort.Vehicle.OnBlowoff {type=Trigger}
 Fort.Vehicle.OnDownshift {type=Trigger}
 Fort.Vehicle.OnUpshift {type=Trigger}
 Fort.Vehicle.Pitch {type=Float}
 Fort.Vehicle.RPM {type=Float}
 Fort.Vehicle.Throttle {type=Float}
 Min Volume {type=Float}
 Movement.Speed {type=Float}
 Movement.SpeedAzimuth {type=Float}
 Movement.SpeedElevation {type=Float}
 Sounds.GearShifts {type=WaveAsset:Array}
 Sounds.Loop {type=WaveAsset:Array}
 Throttle.EQFrequency {type=Float}
 Throttle.Throttle {type=Float}
 UE.Source.OnPlay {type=Trigger}
]
interfaces [Fort.Movement:1.0, Fort.Vehicle:1.0, UE.OutputFormat.Stereo:1.0, UE.Source.OneShot:1.0, UE.Source:1.1]
outputs [
 "UE.OutputFormat.Stereo.Audio:0" {type=Audio}
 "UE.OutputFormat.Stereo.Audio:1" {type=Audio}
 UE.Source.OneShot.OnFinished {type=Trigger}
]
pages [Default {
  input_defaults {
   Fort.Vehicle.Acceleration=0
   Fort.Vehicle.BlowoffStrength=0
   Fort.Vehicle.BonusSpeed=0
   Fort.Vehicle.EnableShifting=null
   Fort.Vehicle.Gear=0
   Fort.Vehicle.IsGrounded="<MetaSoundBuilderResult.SUCCEEDED: 0>"
   Fort.Vehicle.Load=0
   Fort.Vehicle.LowFuel=null
   Fort.Vehicle.NormalSpeed=0
   Fort.Vehicle.OnBlowoff=null
   Fort.Vehicle.OnDownshift=null
   Fort.Vehicle.OnUpshift=null
   Fort.Vehicle.Pitch=0
   Fort.Vehicle.RPM=0
   Fort.Vehicle.Throttle=0
   Min Volume=0.3
   Movement.Speed=0
   Movement.SpeedAzimuth=0
   Movement.SpeedElevation=0
   Throttle.EQFrequency=2000
   Throttle.Throttle=0.024
   UE.Source.OnPlay=null
   Sounds.GearShifts [
    /NitroGameplay/Assets/Audio/Waves/Nitro_State_Vehicle_GearShift_01
    /NitroGameplay/Assets/Audio/Waves/Nitro_State_Vehicle_GearShift_02
    /NitroGameplay/Assets/Audio/Waves/Nitro_State_Vehicle_GearShift_03
    /NitroGameplay/Assets/Audio/Waves/Nitro_State_Vehicle_GearShift_04
    /NitroGameplay/Assets/Audio/Waves/Nitro_State_Vehicle_GearShift_05
    /NitroGameplay/Assets/Audio/Waves/Nitro_State_Vehicle_GearShift_06
    /NitroGameplay/Assets/Audio/Waves/Nitro_State_Vehicle_GearShift_07
   ]
   Sounds.Loop [/NitroGameplay/Assets/Audio/Waves/Nitro_State_Vehicle_Loop]
  }
  nodes [
   n1 = UE.InterpTo:Audio(Target=MapRange.MapRange:Float(Out Range A=^Min Volume, In=^Movement.Speed) {also {defaults {In Range B=6000}}}.Out Value) {also {defaults {Interp Time=0.25}}}
   n2 = AbsoluteValue.Abs:Float(Input=^Fort.Vehicle.Throttle)
   n3 = Clamp.Clamp:Float(In=^Movement.Speed) {also {defaults {Max=6, Min=1}}}
   n4 = UE.Multiply:Float(PrimaryOperand=UE.InterpTo:Audio(Target=UE.Multiply:Float(PrimaryOperand=@n2.Value, AdditionalOperands=@n3.Value).Out) {also {defaults {Interp Time=1}}}.Value) {also {defaults {AdditionalOperands=0.6}}}
   n5 = Clamp.Clamp:Float(In=@n3.Value) {also {defaults {Max=1}}}
   TriggerOnce_0 = UE.Trigger Once(
    Reset=UE.TriggerOnThreshold:Float(In=@n5.Value) {also {defaults {Threshold=0.69, Type=1}}}.Out
    Trigger In=UE.TriggerOnThreshold:Float(In=@n5.Value) {also {defaults {Threshold=0.7}}}.Out
   ) {also {defaults {Start Closed="<MetaSoundBuilderResult.SUCCEEDED: 0>"}}}
   WavePlayer_1 = UE.Wave Player:Stereo(Pitch Shift=@n4.Out, Play=@TriggerOnce_0.Trigger Out) {also {defaults {Loop="<MetaSoundBuilderResult.SUCCEEDED: 0>", Wave Asset=/NitroGameplay/Assets/Audio/Waves/Nitro_State_Vehicle_Loop_Fast}}}
   n6 = UE.InterpTo:Audio(Target=AbsoluteValue.Abs:Float(Input=^Fort.Vehicle.Throttle).Value) {also {defaults {Interp Time=0.25}}}
   ArrayRandomGet_0 = Array.Random Get:WaveAsset:Array(In Array=^Sounds.Loop, Next=^UE.Source.OnPlay) {also {defaults {No Repeats=0}}}
   WavePlayer_0 = UE.Wave Player:Stereo(Play=@ArrayRandomGet_0.On Next, Wave Asset=@ArrayRandomGet_0.Value, Pitch Shift=@n4.Out) {also {defaults {Loop="<MetaSoundBuilderResult.SUCCEEDED: 0>"}}}
   n7 = UE.Multiply:Audio by Float(AdditionalOperands=@n1.Value, PrimaryOperand=@WavePlayer_0.Out Right)
   OnePoleLowPassFilter_1 = UE.One-Pole Low Pass Filter:Audio(Cutoff Frequency=^Throttle.EQFrequency, In=@n7.Out)
   n8 = UE.Multiply:Audio by Float(AdditionalOperands=@n1.Value, PrimaryOperand=@WavePlayer_0.Out Left)
   OnePoleLowPassFilter_0 = UE.One-Pole Low Pass Filter:Audio(Cutoff Frequency=^Throttle.EQFrequency, In=@n8.Out)
   AudioMixerAudioMixerStereo8_0 = "AudioMixer.Audio Mixer (Stereo, 8)"(
    In 2 L=UE.Multiply:Audio by Float(AdditionalOperands=@n1.Value, PrimaryOperand=@WavePlayer_1.Out Left).Out
    In 0 R=UE.Multiply:Audio by Float(AdditionalOperands=@n6.Value, PrimaryOperand=@OnePoleLowPassFilter_1.Out).Out
    In 1 L=@n8.Out
    In 0 L=UE.Multiply:Audio by Float(AdditionalOperands=@n6.Value, PrimaryOperand=@OnePoleLowPassFilter_0.Out).Out
    In 1 R=@n7.Out
    In 2 R=UE.Multiply:Audio by Float(AdditionalOperands=@n1.Value, PrimaryOperand=@WavePlayer_1.Out Right).Out
   )
   /FortMetasound/Patches/RandomSpreadSimple_StartTimePitch() {also {defaults {PitchShiftSpread=15}}}
  ]
  links [
   {dst="@UE.OutputFormat.Stereo.Audio:0", src=AudioMixerAudioMixerStereo8_0.Out L}
   {dst="@UE.OutputFormat.Stereo.Audio:1", src=AudioMixerAudioMixerStereo8_0.Out R}
   {dst="@UE.Source.OneShot.OnFinished", src=WavePlayer_0.On Finished}
  ]
 }]
properties {
 assetImportData=&self:AssetImportData
 attenuationSettings=/NitroGameplay/Assets/Audio/Settings/Att_NitroState_Veh_Loop
 maxDistance=1800
 numChannels=2
 outputFormat=Stereo
 sampleRate=48000
 soundClassObject=&Fort_Vehicles
 volume=0.2
}
