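# Sound cue that plays the nitro warning sound in the NitroGameplay plugin.
# A single wave player plays the Creative DamageAmplifier power-down gadget sound at reduced volume in the Fort_Gadgets sound class.
# Overridden attenuation makes it a short-range, point-like spatial sound with no stereo spread.
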SoundCue /NitroGameplay/Assets/Audio/Nitro_Warning_Cue
root=WavePlayer_1
nodes [{id=WavePlayer_1, properties {soundWaveAssetPtr=/Game/Sounds/Creative/Gadgets/DamageAmplifier/DamageAmplifier_PowerDown_01}}]
properties {
 bOverrideAttenuation=true
 maxDistance=1100
 soundClassObject=&Fort_Gadgets
 volumeMultiplier=0.2
 attenuationOverrides {falloffDistance=1000, stereoSpread=0, attenuationShapeExtents {x=100}}
}
