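# Masked, two-sided lit material for Niagara mesh particles that, by its name, renders goo for a Nitro tire effect.
# Multiplied, tiled cell and Voronoi noise drives color, roughness and specular, alongside an adjustable-strength noise normal map.
# Particle dynamic parameters set U tiling and erosion.
# Eroded noise, shaped by UV gradient falloffs, feeds a vertex-normal displacement and a temporal-AA dithered opacity mask.
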Material /NitroGameplay/Assets/FX/M_NItro_Tire_Goo
graph {
 properties {bUsedWithNiagaraMeshParticles=true, blendMode=BLEND_Masked, pixelDepthOffsetMode=PDOM_Legacy, twoSided=true, usedShadingModels=Default Lit}
 nodes [
  n1 = AppendVector(
   A=LinearInterpolate(Alpha=DynamicParameter() {defaultValue=<rgba 1 0 1 1>, paramNames [Erode,U_Tile,Param3,Param4]}.U_Tile) {constA=1, constB=0.5}
   B=Constant() {r=1}
  )
  n2 = Multiply(A=Multiply(A=TextureCoordinate(), B=Constant2Vector() {g=1, r=1}), B=@n1)
  n3 = Multiply(A=Multiply(A=TextureCoordinate(), B=Constant2Vector() {g=3, r=2}), B=@n1)
  n4 = Multiply(A=TextureCoordinate(), B=@n1)
  n5 = Multiply(A=Multiply(A=TextureCoordinate()) {constB=2}, B=@n1)
  n6 = Multiply(
   A=Multiply(
    A=TextureSample(UVs=@n2) {samplerType=SAMPLERTYPE_Grayscale, texture=/Game/Weapons/FORT_Melee/Pickaxe_Cube_Ninja_Male/FX/T_CellNoise}.R
    B=TextureSample(UVs=@n3) {samplerType=SAMPLERTYPE_Grayscale, texture=&T_CellNoise}.R
   )
   B=Multiply(
    A=TextureSample(UVs=@n4) {samplerType=SAMPLERTYPE_LinearGrayscale, texture=/Game/Packages/Fortress_Sky/SkyDome/Master/T_SkyDome_Cloud_Voroni}.R
    B=TextureSample(UVs=@n5) {samplerType=SAMPLERTYPE_LinearGrayscale, texture=&T_SkyDome_Cloud_Voroni}.R
   )
  ) {constA=1}
  LinearInterpolate(
   A=VectorParameter() {parameterName=ColorLow, defaultValue=<rgba 0.026 0.0059 0.0001 1>}.RGB
   B=VectorParameter() {parameterName=ColorHigh, defaultValue=<rgba 0.1302 0.0286 0.0025 1>}.RGB
   Alpha=Multiply(A=@n6) {constB=50}
  )
  n7 = TextureCoordinate()
  n8 = Constant2Vector() {g=5, r=2}
  n9 = TextureSample(UVs=Multiply(A=Multiply(A=@n7, B=@n8), B=@n1)) {samplerType=SAMPLERTYPE_Normal, texture=/Game/Characters/Player/Female/Medium/Bodies/F_MED_Bistro_Astronaut/FX/T_GaussianNoise_Normal}
  MaterialFunctionCall(Normal=@n9.RGB, Flatness=ScalarParameter() {defaultValue=-1, parameterName=NormalStrength}) {materialFunction=&FlattenNormal}
  n10 = Multiply(
   A=@n6
   B=LinearInterpolate(B=Constant() {r=500}, Alpha=DynamicParameter() {defaultValue=<rgba 1 0 1 1>, paramNames [Erode,U_Tile,Param3,Param4]}.Erode) {constAlpha=1, constB=4}
  )
  n11 = MaterialFunctionCall() {materialFunction=&LinearGradient}
  n12 = Multiply(A=@n11.UGradient) {constB=0.5}
  n13 = Multiply(A=Subtract(A=Sine(None=@n12), B=Constant()), B=Constant() {r=2}) {constB=2}
  n14 = Power(Base=@n11.VGradient, Exp=Constant() {r=2})
  n15 = Constant() {r=0.5}
  n16 = Power(Base=Multiply(A=SphereMask(A=@n14, B=@n15) {attenuationRadius=0.5, hardnessPercent=0}, B=Constant() {r=1}) {constB=2}, Exp=Constant() {r=2})
  n17 = Multiply(A=Saturate(None=@n10), B=Saturate(None=Multiply(A=@n13, B=@n16)))
  Multiply(A=VertexNormalWS(), B=LinearInterpolate(A=Constant() {r=-2}, B=Constant() {r=2}, Alpha=@n17) {constA=-1})
  Multiply(A=LinearInterpolate(Alpha=@n6) {constA=0.4, constB=0.1}, B=ScalarParameter() {defaultValue=1.2, parameterName=Roughness})
  Multiply(A=ScalarParameter() {defaultValue=8, parameterName=Specular}, B=LinearInterpolate(Alpha=@n6))
  Power(Base=Saturate(None=@n10)) {constExponent=4}
  n18 = ScalarParameter() {defaultValue=5, parameterName=OverallOpacityMul}
  Saturate(None=MaterialFunctionCall(Alpha Threshold=Multiply(A=@n17, B=@n18)) {materialFunction=&DitherTemporalAA_Fortnite}.Result)
  Saturate(None=Subtract(A=@n10, B=Constant() {r=2}))
  ScalarParameter() {defaultValue=0.5, parameterName=Metallic}
 ]
}
