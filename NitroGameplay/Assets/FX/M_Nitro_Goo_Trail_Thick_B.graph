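# Translucent unlit Niagara ribbon material for a thick, orange Nitro goo trail.
# Emissive blends two orange tones along a UV gradient mask, tinted by particle color through an emissive adjustment function.
# Opacity multiplies gradient masks by panned directional noise on UV1, subtracts an erosion threshold, then boosts the result and applies particle alpha and view-angle fade.
# Dynamic parameters drive noise tiling (Tile) and erosion (Erode).
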Material /NitroGameplay/Assets/FX/M_Nitro_Goo_Trail_Thick_B
graph {
 properties {bUsedWithNiagaraRibbons=true, blendMode=BLEND_Translucent, pixelDepthOffsetMode=PDOM_Legacy, shadingModel=MSM_Unlit, usedShadingModels=Unlit}
 nodes [
  n1 = Constant3Vector() {constant=<rgba 1 0.3072 0 1>}
  n2 = Constant3Vector() {constant=<rgba 1 0.0648 0 1>}
  n3 = MaterialFunctionCall() {materialFunction=&LinearGradient}
  n4 = OneMinus(None=@n3.UGradient)
  n5 = Power(Base=Multiply(A=@n4, B=@n3.VGradient), Exp=Constant() {r=0.5}) {constExponent=0.5}
  n6 = Constant() {r=0.2}
  n7 = Saturate(None=Multiply(A=Subtract(A=@n5, B=@n6), B=Constant() {r=2}))
  n8 = ParticleColor()
  MaterialFunctionCall(Emissive=Multiply(A=LinearInterpolate(A=@n1, B=@n2, Alpha=@n7), B=@n8.RGB)) {materialFunction=&MF_FX_EmissiveAdjustments}
  n9 = Multiply(A=OneMinus(None=@n3.VGradient))
  n10 = Multiply(A=Multiply(A=@n4, B=@n3.VGradient), B=Saturate(None=@n9))
  n11 = AppendVector(A=ScalarParameter() {defaultValue=1.5, parameterName=X_Tile}, B=Constant() {r=0.3})
  n12 = LinearInterpolate(Alpha=DynamicParameter() {defaultValue=<rgba 1 0 1 1>, paramNames [Tile,Erode,Param3,Param4]}.Tile) {constA=1, constB=3}
  n13 = Panner(
   Coordinate=Multiply(A=TextureCoordinate() {coordinateIndex=1}, B=Divide(A=@n11, B=@n12) {constB=1.5})
   Speed=AppendVector(A=ScalarParameter() {defaultValue=0.5, parameterName=X_PanSpeed}, B=Constant() {r=0.3})
  ) {speedX=0.5, speedY=0.3}
  n14 = Multiply(A=@n3.UGradient, B=@n9)
  n15 = Saturate(None=Subtract(
    A=TextureSample(UVs=@n13) {samplerType=SAMPLERTYPE_LinearGrayscale, texture=/Game/Effects/Fort_Effects/Emotes/UltraLight/Texture/T_DirectionalNoise_Greyscale}.R
    B=Divide(A=@n14) {constB=1.2}
   ) {constB=0})
  n16 = DynamicParameter() {defaultValue=<rgba 1 0 1 1>, paramNames [Tile,Erode,Param3,Param4]}
  n17 = Saturate(None=Subtract(A=Multiply(A=@n10, B=@n15), B=LinearInterpolate(Alpha=@n16.Erode) {constB=0.18}))
  n18 = Constant() {r=15}
  Multiply(
   A=Multiply(A=@n8.A, B=Multiply(A=@n17, B=@n18))
   B=MaterialFunctionCall("FadeSensitivity (optional)"=ScalarParameter() {defaultValue=4, parameterName=FadeBasedOnViewAngle}) {materialFunction=&FadeBasedOnViewAngle}.Result
  )
 ]
}
