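# Translucent unlit Niagara ribbon material for nitro tire trails, blending dark purple with noise-varied orange and yellow, tinted by particle color.
# Two noise textures share UVs shifted by the UV_Offset dynamic parameter.
# Opacity combines a UV1 lengthwise fade and sine edge falloff, then subtracts fibre noise and the Erode dynamic parameter for a ragged dissolve.
# Particle alpha and an Opacity scalar scale the result.
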Material /NitroGameplay/Assets/FX/M_Nitro_Tire_Trail
graph {
 properties {bUsedWithNiagaraRibbons=true, blendMode=BLEND_Translucent, pixelDepthOffsetMode=PDOM_Legacy, shadingModel=MSM_Unlit, usedShadingModels=Unlit}
 nodes [
  n1 = Constant3Vector() {constant=<rgba 0.0333 0.0039 0.0347 1>}
  n2 = DynamicParameter() {defaultValue=<rgba 1 0 1 1>, paramNames [UV_Offset,Erode,Param3,Param4]}
  n3 = Multiply(A=Add(A=TextureCoordinate(), B=@n2.UV_Offset), B=Constant2Vector() {g=1, r=1})
  n4 = LinearInterpolate(
   A=Constant3Vector() {constant=<rgba 1 0.1017 0 1>}
   B=Constant3Vector() {constant=<rgba 1 0.5402 0.071 1>}
   Alpha=TextureSample(UVs=@n3) {texture=/Game/Weapons/FORT_Melee/Pickaxe_MadameMoth/FX/T_Fibre_Noise}.R
  )
  n5 = TextureSample(UVs=@n3) {samplerType=SAMPLERTYPE_Grayscale, texture=/Game/Global/Textures/Default/Noise/TilingNoise06}
  n6 = ParticleColor()
  MaterialFunctionCall(Emissive=Multiply(A=LinearInterpolate(A=@n1, B=@n4, Alpha=@n5.R), B=@n6.RGB)) {materialFunction=&MF_FX_EmissiveAdjustments}
  MaterialFunctionCall() {materialFunction=&DebugTimeSine}
  n7 = MaterialFunctionCall(UV Channel=TextureCoordinate() {coordinateIndex=1}) {materialFunction=&LinearGradient}
  n8 = OneMinus(None=@n7.UGradient)
  n9 = Multiply(A=@n7.VGradient) {constB=0.5}
  n10 = Saturate(None=Multiply(A=Sine(None=@n9), B=Constant() {r=1}))
  n11 = Divide(A=@n3) {constB=1.4}
  n12 = Subtract(
   A=Subtract(A=Multiply(A=@n8, B=@n10), B=TextureSample(UVs=@n11) {texture=&T_Fibre_Noise}.R)
   B=@n2.Erode
  )
  Multiply(A=Multiply(A=@n6.A, B=Saturate(None=@n12)), B=ScalarParameter() {defaultValue=1.5, parameterName=Opacity})
 ]
}
